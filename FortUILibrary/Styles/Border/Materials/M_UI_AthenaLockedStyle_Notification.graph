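# Translucent UI material that draws a blue, angled notification background for the Athena locked style.
# An aspect-corrected, rotated triangle grid texture blends two colors, fading into a horizontal gradient across the width, then scaled by Darken.
# A rotated, offsettable flare band, modulated by the triangles, adds a third color, which also paints a slanted trim at one side.
# Rotated UV ramps with steep falloffs cut the trim and the slanted opacity edge procedurally.
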Material /FortUILibrary/Styles/Border/Materials/M_UI_AthenaLockedStyle_Notification
graph {
 properties {
  bUsedWithStaticMesh=false
  blendMode=BLEND_Translucent
  materialDomain=MD_UI
  pixelDepthOffsetMode=PDOM_Legacy
  usedShadingModels=Default Lit
 }
 nodes [
  n1 = VectorParameter() {parameterName=RGBA-1, defaultValue=<rgba 0.0033 0.1195 0.6867 1>}
  n2 = VectorParameter() {parameterName=RGBA-2, defaultValue=<rgba 0.0075 0.1826 0.6927 1>}
  n3 = ComponentMask(None=MaterialFunctionCall() {materialFunction=&GetUserInterfaceUV}.Normalized UV) {r=true}
  n4 = MaterialFunctionCall() {materialFunction=&GetUserInterfaceUV}
  n5 = MaterialFunctionCall(Float2=@n4.Pixel Size) {materialFunction=&BreakOutFloat2Components}
  n6 = Max(A=@n5.G)
  n7 = Multiply(A=AppendVector(A=Divide(A=@n5.R, B=@n6), B=Constant() {r=1}), B=@n4.Normalized UV)
  n8 = ScalarParameter() {defaultValue=-0.0278, parameterName=Tri Angle}
  n9 = TextureSample(UVs=Multiply(A=Rotator(Coordinate=@n7, Time=@n8) {speed=6.2832}, B=ScalarParameter() {defaultValue=0.35, parameterName=Tri Tile Amount})) {texture=/Game/Global/Textures/Default/Masks_and_Gradients/T-TriangleGrid}
  n10 = LinearInterpolate(A=LinearInterpolate(A=@n1.RGB, B=@n2.RGB, Alpha=@n3), B=LinearInterpolate(A=@n1.RGB, B=@n2.RGB, Alpha=@n9.R), Alpha=OneMinus(None=@n3))
  n11 = ScalarParameter() {defaultValue=1.35, parameterName=Darken}
  n12 = MaterialFunctionCall() {materialFunction=&GetUserInterfaceUV}
  n13 = Constant() {r=-0.0278}
  n14 = ScalarParameter() {parameterName=Flare-OffSet}
  n15 = Abs(None=Add(A=Rotator(Coordinate=@n12.Normalized UV, Time=@n13) {speed=6.2832}, B=LinearInterpolate(Alpha=@n14) {constA=0.5, constB=-1.5}) {constB=-0.5})
  n16 = Subtract(A=ComponentMask(None=OneMinus(None=@n15)) {r=true}, B=ScalarParameter() {defaultValue=0.25, parameterName=Flare-Clamp})
  n17 = ScalarParameter() {defaultValue=1, parameterName=Flare-Edge}
  n18 = LinearInterpolate(B=Clamp(None=Multiply(A=@n16, B=@n17)), Alpha=Clamp(None=@n9.R) {minDefault=0.5})
  n19 = VectorParameter() {parameterName=RGBA-3, defaultValue=<rgba 0.2831 0.6376 0.9823 1>}
  n20 = TextureCoordinate()
  n21 = Rotator(Coordinate=@n20, Time=ScalarParameter() {defaultValue=-0.035, parameterName=TrimAngle}) {speed=6.2832}
  n22 = Subtract(A=OneMinus(None=ComponentMask(None=@n21) {r=true}), B=ScalarParameter() {defaultValue=0.75, parameterName=TrimOffset})
  n23 = ScalarParameter() {defaultValue=32, parameterName=TrimFalloff}
  LinearInterpolate(A=Add(A=Multiply(A=@n10, B=@n11), B=Multiply(A=@n18, B=@n19.RGB)), B=@n19.RGB, Alpha=Clamp(None=Multiply(A=@n22, B=@n23)))
  n24 = Rotator(Coordinate=@n20, Time=ScalarParameter() {defaultValue=-0.0222, parameterName=EdgeAngle}) {speed=6.2832}
  n25 = Multiply(
   A=Subtract(A=ComponentMask(None=@n24) {r=true}, B=ScalarParameter() {defaultValue=0.07, parameterName=EdgeOffset})
   B=ScalarParameter() {defaultValue=32, parameterName=EdgeFalloff}
  )
  Reroute(None=Multiply(A=Clamp(None=@n25), B=ScalarParameter() {defaultValue=0.85, parameterName=Alpha}))
 ]
}
