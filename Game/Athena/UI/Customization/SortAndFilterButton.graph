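# Common UI button widget for the Customization sort-and-filter blade, with a text label and a selection pip.
# Hover animates background and content color, showing a pip outline while unselected.
# Selection plays Selected or Deselected animations only while the blade is open; HandleBladeOpenChanged replays changes missed while closed by comparing against the last visible state.
# Helper functions set the label, toggle a highlight border and snap animations to reset states.
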WidgetBlueprint /Game/Athena/UI/Customization/SortAndFilterButton
animations [
 Hover {
  binding_count=2
  length_seconds=0.2
  object=&self:Hover
  sequence {
   bindings [
    Border_Container {
     id=#1
     possessed_class=Border
     tracks [
      {
       type=MovieSceneColorTrack
       display_name=Brush Color
       sections [{
         type=MovieSceneColorSection
         channels [
          Color.R {type=MovieSceneScriptingFloatChannel, const [1,1]}
          Color.G {type=MovieSceneScriptingFloatChannel, const [1,1]}
          Color.B {type=MovieSceneScriptingFloatChannel, const [1,1]}
          Color.A {type=MovieSceneScriptingFloatChannel, keys [[0,0],[7.9767,1]]}
         ]
        }]
      }
      {
       type=MovieSceneWidgetMaterialTrack
       sections [{type=MovieSceneParameterSection, channels [LeftBar_Position {type=MovieSceneScriptingFloatChannel, keys [[0,0.05],[20,-0.26]]}]}]
      }
     ]
    }
    Border_FontColorChange {
     id=#2
     possessed_class=Border
     tracks [{
       type=MovieSceneColorTrack
       display_name=Content Color and Opacity
       sections [{
         type=MovieSceneColorSection
         channels [
          Color.R {type=MovieSceneScriptingFloatChannel, keys [[0,1],[7.9767,0.0003]]}
          Color.G {type=MovieSceneScriptingFloatChannel, keys [[0,1],[7.9767,0.0033]]}
          Color.B {type=MovieSceneScriptingFloatChannel, keys [[0,1],[7.9767,0.0844]]}
          Color.A {type=MovieSceneScriptingFloatChannel, const [1,2]}
         ]
        }]
      }]
    }
   ]
   display_rate {numerator=100}
  }
 }
 Selected {
  binding_count=3
  length_seconds=0.3
  object=&self:Selected
  sequence {bindings [
    Text_ButtonAction {
     id=#3
     possessed_class=CommonTextBlock
     tracks [{
       type=MovieSceneColorTrack
       display_name=Color and Opacity
       sections [{
         type=MovieSceneColorSection
         channels [
          Color.R {type=MovieSceneScriptingFloatChannel, keys [[0,0.0242],[2.9997,1]]}
          Color.G {type=MovieSceneScriptingFloatChannel, keys [[0,0.5906],[2.9997,1]]}
          Color.B {type=MovieSceneScriptingFloatChannel, keys [[0,0.8148],[2.9997,1]]}
          Color.A {type=MovieSceneScriptingFloatChannel, const [1,2]}
         ]
        }]
      }]
    }
    Pip {
     id=#4
     possessed_class=Image
     tracks [
      {
       type=MovieSceneColorTrack
       display_name=Color and Opacity
       sections [{
         type=MovieSceneColorSection
         channels [
          Color.R {type=MovieSceneScriptingFloatChannel, keys [[0,0.0242],[2.9997,1]]}
          Color.G {type=MovieSceneScriptingFloatChannel, keys [[0,0.5906],[2.9997,1]]}
          Color.B {type=MovieSceneScriptingFloatChannel, keys [[0,0.8148],[2.9997,1]]}
          Color.A {type=MovieSceneScriptingFloatChannel, const [1,2]}
         ]
        }]
      }
      {
       type=MovieSceneFloatTrack
       display_name=Render Opacity
       sections [{type=MovieSceneFloatSection, channels [None {type=MovieSceneScriptingFloatChannel, keys [[0,0],[2.9997,1]]}]}]
      }
      {
       type=MovieScene2DTransformTrack
       sections [{
         type=MovieScene2DTransformSection
         channels [
          Scale.X {type=MovieSceneScriptingFloatChannel, keys [[0,0.75],[4,1.5],[6,1]]}
          Scale.Y {type=MovieSceneScriptingFloatChannel, keys [[0,0.75],[4,1.5],[6,1]]}
         ]
        }]
      }
     ]
    }
    Pip_Outline {
     id=#5
     possessed_class=Image
     tracks [
      {
       type=MovieSceneFloatTrack
       display_name=Render Opacity
       sections [{type=MovieSceneFloatSection, channels [None {type=MovieSceneScriptingFloatChannel, keys [[0,1],[2.9997,0]]}]}]
      }
      {
       type=MovieScene2DTransformTrack
       sections [{
         type=MovieScene2DTransformSection
         channels [
          Scale.X {type=MovieSceneScriptingFloatChannel, keys [[0,1],[2,0.5],[2.9997,1]]}
          Scale.Y {type=MovieSceneScriptingFloatChannel, keys [[0,1],[2,0.5],[2.9997,1]]}
         ]
        }]
      }
     ]
    }
   ]}
 }
 Deselected {
  binding_count=2
  length_seconds=0.15
  object=&self:Deselected
  sequence {bindings [
    Text_ButtonAction {
     id=#3
     possessed_class=CommonTextBlock
     tracks [{
       type=MovieSceneColorTrack
       display_name=Color and Opacity
       sections [{
         type=MovieSceneColorSection
         channels [
          Color.R {type=MovieSceneScriptingFloatChannel, keys [[0,1],[3,0.0242]]}
          Color.G {type=MovieSceneScriptingFloatChannel, keys [[0,1],[3,0.5906]]}
          Color.B {type=MovieSceneScriptingFloatChannel, keys [[0,1],[3,0.8148]]}
          Color.A {type=MovieSceneScriptingFloatChannel, const [1,2]}
         ]
        }]
      }]
    }
    Pip {
     id=#4
     possessed_class=Image
     tracks [
      {
       type=MovieSceneColorTrack
       display_name=Color and Opacity
       sections [{
         type=MovieSceneColorSection
         channels [
          Color.R {type=MovieSceneScriptingFloatChannel, keys [[0,1],[3,0.0242]]}
          Color.G {type=MovieSceneScriptingFloatChannel, keys [[0,1],[3,0.5906]]}
          Color.B {type=MovieSceneScriptingFloatChannel, keys [[0,1],[3,0.8148]]}
          Color.A {type=MovieSceneScriptingFloatChannel, const [1,2]}
         ]
        }]
      }
      {
       type=MovieSceneFloatTrack
       display_name=Render Opacity
       sections [{type=MovieSceneFloatSection, channels [None {type=MovieSceneScriptingFloatChannel, keys [[0,1],[3,0]]}]}]
      }
     ]
    }
   ]}
 }
 ShowPipOutline {
  binding_count=1
  length_seconds=0.0798
  object=&self:ShowPipOutline
  sequence {bindings [Pip_Outline {
     id=#6
     possessed_class=Image
     tracks [{
       type=MovieSceneFloatTrack
       display_name=Render Opacity
       sections [{type=MovieSceneFloatSection, channels [None {type=MovieSceneScriptingFloatChannel, keys [[0,0],[1.5953,1]]}]}]
      }]
    }]}
 }
]
graphs [
 EventGraph {nodes [
   n1 = HandleBladeOpenChanged!() -> (OutputDelegate, then, bBladeOpen:bool=~) {is=title K2Node_CustomEvent}
   > n2 = Set bIsBladeIsOpen(bIsBladeIsOpen:bool=@n1.bBladeOpen) -> (Output_Get:bool)
   n3 = Get bLastVisibleSelectedValue() -> (bLastVisibleSelectedValue:bool)
   n4 = Get isSelected() -> (isSelected:bool)
   > Branch(Condition=AND Boolean(A=@n2, B=XOR Boolean!(A:bool=@n4, B:bool=@n3) -> (ReturnValue:bool)))
   > PlayAnimationForward(InAnimation=Select(
     Option 0:Widget Animation Object Reference=Get Deselected() -> (Deselected:Widget Animation Object Reference)
     Option 1:Widget Animation Object Reference=Get Selected() -> (Selected:Widget Animation Object Reference)
     Index=Get isSelected()
    ) -> (ReturnValue:Widget Animation Object Reference))
   > Set bLastVisibleSelectedValue(bLastVisibleSelectedValue:bool=Get isSelected()) -> (Output_Get:bool)
   n5 = Event PreConstruct()
   > n6 = Branch(Condition=@n5.IsDesignTime)
   SetText!(execute=[@n7.then, @n6.then], self:Self Object Reference, Text:Text=Get Button Description() -> (Button Description:Text)) -> (then)
   > Set bLastVisibleSelectedValue(bLastVisibleSelectedValue=Get isSelected())
   Event Construct()
   > n7 = Parent: Construct()
   Event OnMouseLeave!() -> (OutputDelegate, then, MouseEvent:"Pointer Event Structure (by ref)") {is=title K2Node_Event}
   n8 = Get Hover() -> (Hover:Widget Animation Object Reference)
   > PlayAnimationReverse(InAnimation=@n8, PlaybackSpeed=2)
   n9 = Get isSelected()
   > n10 = Branch(Condition=@n9)
   n11 = Get ShowPipOutline() -> (ShowPipOutline:Widget Animation Object Reference)
   PlayAnimationReverse(execute=@n10.else, InAnimation=@n11)
   Event On Selected()
   > Set isSelected(isSelected:bool=true) -> (Output_Get:bool)
   > Branch(Condition=Get bIsBladeIsOpen() -> (bIsBladeIsOpen:bool))
   > PlayAnimationForward(InAnimation=Get Selected())
   > Set bLastVisibleSelectedValue(bLastVisibleSelectedValue=true)
   Event OnMouseEnter()
   > PlayAnimationForward(InAnimation=@n8)
   > n12 = Branch(Condition=@n9)
   PlayAnimationForward(execute=@n12.else, InAnimation=@n11)
   Event On Deselected()
   > Set isSelected(isSelected=false)
   > Branch(Condition=Get bIsBladeIsOpen())
   > PlayAnimationForward(InAnimation=Get Deselected())
   > Set bLastVisibleSelectedValue(bLastVisibleSelectedValue=false)
   Branch(Condition=Get isSelected())
   > PlayAnimationForward(InAnimation=Get ShowPipOutline())
   ResetAnimations()
   Event Tick()
  ]}
 ResetAnimations {nodes [
   ResetAnimations() {is=title K2Node_FunctionEntry}
   > SnapToEndOfAnimation!(execute, self, Animation:Widget Animation Object Reference=Get Hover(), PlayMode:EUMGSequencePlayMode Enum=Reverse) -> (then)
   > SnapToEndOfAnimation(Animation=Get Deselected())
   > SnapToEndOfAnimation(Animation=Get ShowPipOutline(), PlayMode=Reverse)
  ]}
 SetHighlightBorderVisibility {nodes [
   n1 = SetHighlightBorderVisibility!() -> (then, bVisible:bool=~) {is=title K2Node_FunctionEntry}
   > SetVisibility(
    self=Get Border_Highlight() -> (Border_Highlight:Border Object Reference)
    InVisibility=Select(Option 0:ESlateVisibility Enum=Collapsed, Option 1:ESlateVisibility Enum=SelfHitTestInvisible, Index=@n1) -> (ReturnValue:ESlateVisibility Enum)
   )
  ]}
 SetText {nodes [
   n1 = SetText!() -> (then, Text:Text) {is=title K2Node_FunctionEntry}
   > n2 = Set Button Description(Button Description:Text=@n1) -> (Output_Get:Text)
   > "SetText (Text)"(self=Get Text_ButtonAction() -> (Text_ButtonAction:Common Text Object Reference), InText=@n2)
  ]}
 SnapToEndOfAnimation {nodes [
   n1 = SnapToEndOfAnimation!() -> (then, Animation:Widget Animation Object Reference, PlayMode:EUMGSequencePlayMode Enum=Forward) {is=title K2Node_FunctionEntry}
   > PlayAnimation(InAnimation=@n1, StartAtTime=GetEndTime(self=@n1), PlayMode=@n1.PlayMode)
  ]}
]
variables [
 bIsBladeIsOpen {type=boolean}
 bLastVisibleSelectedValue {type=boolean}
 Button Description {type=string, default="{TEXT}"}
 isSelected {type=boolean}
]
widgets {
 info {inheritedWidgetCount=1, parentClass=&CommonButtonLegacy, rootWidgetClass=&Border, widgetCount=10}
 widgets [
  Border_Container {
   widgetClassPath=&Border
   widget=Border_Container
   properties {
    background {
     bIsDynamicallyLoaded=false
     drawAs=Box
     imageType=NoImage
     mirroring=NoMirror
     resourceName=None
     resourceObject=/Game/Athena/UI/Customization/Art/M_UI_SortAndFilter_Button_Selection_Inst
     tiling=NoTile
     imageSize=<64 64>
     margin=<ltrb 0.5 0 0.5 0>
     outlineSettings {
      bUseBrushTransparency=false
      roundingType=HalfHeightRadius
      width=0
      color {colorUseRule=UseColor_Specified, specifiedColor=<rgba 0 0 0 0>}
      cornerRadii=<0 0 0 0>
     }
     tintColor {colorUseRule=UseColor_Specified, specifiedColor=<rgba 1 1 1 1>}
     uVRegion {bIsValid=false, max=<0 0>, min=<0 0>}
    }
    brushColor=<rgba 1 1 1 0>
    padding=<ltrb 0 0 0 0>
   }
  }
  Overlay_0 {widgetClassPath=&Overlay, parent=Border_Container, widget=Overlay_0, slot {class=BorderSlot, properties {padding=<ltrb 0 0 0 0>}}}
  Border_Highlight {
   widgetClassPath=&Border
   bIsVariable=true
   parent=Overlay_0
   widget=Border_Highlight
   properties {
    visibility=Collapsed
    background {
     bIsDynamicallyLoaded=false
     drawAs=Image
     imageType=NoImage
     mirroring=NoMirror
     resourceName=None
     resourceObject=null
     tiling=NoTile
     imageSize=<32 32>
     margin=<ltrb 0 0 0 0>
     outlineSettings {
      bUseBrushTransparency=false
      roundingType=HalfHeightRadius
      width=0
      color {colorUseRule=UseColor_Specified, specifiedColor=<rgba 0 0 0 0>}
      cornerRadii=<0 0 0 0>
     }
     tintColor {colorUseRule=UseColor_Specified, specifiedColor=<rgba 0 0.0823 0.9216 1>}
     uVRegion {bIsValid=false, max=<0 0>, min=<0 0>}
    }
   }
   slot {class=OverlaySlot, properties {horizontalAlignment=HAlign_Fill, padding=<ltrb 20 0 20 0>}}
  }
  InputActionWidget {
   widgetClassPath=&CommonActionWidget
   bInherited=true
   bIsVariable=true
   parent=Overlay_0
   widget=InputActionWidget
   properties {
    progressMaterialParam=Percentage
    progressMaterialBrush {
     bIsDynamicallyLoaded=false
     drawAs=Image
     imageType=NoImage
     mirroring=NoMirror
     resourceName=None
     resourceObject=/Game/UI/Materials/HollowCircularProgressBar
     tiling=NoTile
     imageSize=<36 36>
     margin=<ltrb 0 0 0 0>
     outlineSettings {
      bUseBrushTransparency=false
      roundingType=HalfHeightRadius
      width=0
      color {colorUseRule=UseColor_Specified, specifiedColor=<rgba 0 0 0 0>}
      cornerRadii=<0 0 0 0>
     }
     tintColor {colorUseRule=UseColor_Specified, specifiedColor=<rgba 1 1 0 1>}
     uVRegion {bIsValid=false, max=<0 0>, min=<0 0>}
    }
   }
   slot {class=OverlaySlot, properties {verticalAlignment=VAlign_Center, padding=<ltrb 8 0 0 0>}}
  }
  Border_FontColorChange {
   widgetClassPath=&Border
   parent=Overlay_0
   widget=Border_FontColorChange
   properties {
    background {
     bIsDynamicallyLoaded=false
     drawAs=NoDrawType
     imageType=NoImage
     mirroring=NoMirror
     resourceName=None
     resourceObject=null
     tiling=NoTile
     imageSize=<32 32>
     margin=<ltrb 0 0 0 0>
     outlineSettings {
      bUseBrushTransparency=false
      roundingType=HalfHeightRadius
      width=0
      color {colorUseRule=UseColor_Specified, specifiedColor=<rgba 0 0 0 0>}
      cornerRadii=<0 0 0 0>
     }
     tintColor {colorUseRule=UseColor_Specified, specifiedColor=<rgba 1 1 1 1>}
     uVRegion {bIsValid=false, max=<0 0>, min=<0 0>}
    }
    padding=<ltrb 0 0 0 0>
   }
   slot {class=OverlaySlot, properties {horizontalAlignment=HAlign_Fill, verticalAlignment=VAlign_Fill, padding=<ltrb 46 0 0 0>}}
  }
  HorizontalBox_1 {
   widgetClassPath=&HorizontalBox
   parent=Border_FontColorChange
   widget=HorizontalBox_1
   slot {class=BorderSlot, properties {padding=<ltrb 0 0 0 0>}}
  }
  Overlay_1 {
   widgetClassPath=&Overlay
   parent=HorizontalBox_1
   widget=Overlay_1
   slot {class=HorizontalBoxSlot, properties {verticalAlignment=VAlign_Center}}
  }
  Pip_Outline {
   widgetClassPath=&Image
   parent=Overlay_1
   widget=Pip_Outline
   properties {
    renderOpacity=0
    visibility=HitTestInvisible
    brush {
     bIsDynamicallyLoaded=false
     drawAs=Image
     imageType=NoImage
     mirroring=NoMirror
     resourceName=None
     resourceObject=/Game/Athena/UI/Customization/Art/M_UI_SortAndFilter_PipOutline
     tiling=NoTile
     imageSize=<26 26>
     margin=<ltrb 0 0 0 0>
     outlineSettings {
      bUseBrushTransparency=false
      roundingType=HalfHeightRadius
      width=0
      color {colorUseRule=UseColor_Specified, specifiedColor=<rgba 0 0 0 0>}
      cornerRadii=<0 0 0 0>
     }
     tintColor {colorUseRule=UseColor_Specified, specifiedColor=<rgba 1 1 1 1>}
     uVRegion {bIsValid=false, max=<0 0>, min=<0 0>}
    }
    colorAndOpacity=<rgba 0.0242 0.5906 0.8148 1>
   }
   slot {class=OverlaySlot, properties {horizontalAlignment=HAlign_Fill, verticalAlignment=VAlign_Fill, padding=<ltrb -4 -4 -4 -4>}}
  }
  Pip {
   widgetClassPath=&Image
   parent=Overlay_1
   widget=Pip
   properties {
    renderOpacity=0
    visibility=HitTestInvisible
    brush {
     bIsDynamicallyLoaded=false
     drawAs=Image
     imageType=NoImage
     mirroring=NoMirror
     resourceName=None
     resourceObject=/Game/Athena/UI/Customization/Art/M_UI_SortAndFilter_Pip
     tiling=NoTile
     imageSize=<26 26>
     margin=<ltrb 0 0 0 0>
     outlineSettings {
      bUseBrushTransparency=false
      roundingType=HalfHeightRadius
      width=0
      color {colorUseRule=UseColor_Specified, specifiedColor=<rgba 0 0 0 0>}
      cornerRadii=<0 0 0 0>
     }
     tintColor {colorUseRule=UseColor_Specified, specifiedColor=<rgba 1 1 1 1>}
     uVRegion {bIsValid=false, max=<0 0>, min=<0 0>}
    }
    colorAndOpacity=<rgba 0.0242 0.5906 0.8148 1>
   }
   slot {class=OverlaySlot, properties {horizontalAlignment=HAlign_Fill, verticalAlignment=VAlign_Fill}}
  }
  Text_ButtonAction {
   widgetClassPath=&CommonTextBlock
   bIsVariable=true
   parent=HorizontalBox_1
   widget=Text_ButtonAction
   properties {
    mobileFontSizeMultiplier=1.3
    scrollStyle=&TextScrollStyle-Base_C
    text=ACTION
    textTransformPolicy=ToUpper
    colorAndOpacity {colorUseRule=UseColor_Specified, specifiedColor=<rgba 0.0242 0.5906 0.8148 1>}
    font {
     bForceMonospaced=false
     bMaterialIsStencil=false
     fontMaterial=null
     fontObject=&BurbankSmall
     letterSpacing=0
     monospacedWidth=1
     size=15
     skewAmount=0
     typefaceFontName=BLACK
     outlineSettings {
      bApplyOutlineToDropShadows=false
      bMiteredCorners=false
      bSeparateFillAlpha=false
      outlineMaterial=null
      outlineSize=0
      outlineColor=<rgba 0 0 0 1>
     }
    }
    shadowOffset=<0 0>
   }
   slot {class=HorizontalBoxSlot, properties {verticalAlignment=VAlign_Center, padding=<ltrb 25 0 0 0>, size {sizeRule=Fill, value=1}}}
  }
 ]
}
