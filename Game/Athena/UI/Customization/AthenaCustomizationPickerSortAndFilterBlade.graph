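# Widget Blueprint for the sort-and-filter blade of the Athena customization picker (locker), listing sort and filter buttons in a scroll box.
# Open and close play a slide, fade and blur animation, with close reversing from mid-play if interrupted, and notify every button.
# Native parent events set button names, highlights and archive-mode headings.
# Custom navigation functions link the last sort button and first filter button.
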WidgetBlueprint /Game/Athena/UI/Customization/AthenaCustomizationPickerSortAndFilterBlade
animations [Anim_Open {
  binding_count=2
  length_seconds=0.2
  object=&self:Anim_Open
  sequence {
   bindings [
    BackgroundBlur_30 {
     id=#1
     possessed_class=BackgroundBlur
     tracks [{
       type=MovieSceneFloatTrack
       display_name=Blur Strength
       sections [{type=MovieSceneFloatSection, channels [None {type=MovieSceneScriptingFloatChannel, keys [[0,0],[20,2]]}]}]
      }]
    }
    Border {
     id=#2
     possessed_class=Border
     tracks [
      {
       type=MovieScene2DTransformTrack
       sections [{
         type=MovieScene2DTransformSection
         channels [
          Translation.X {type=MovieSceneScriptingFloatChannel, keys [[0,-128],[20,0]]}
          Translation.Y {type=MovieSceneScriptingFloatChannel}
         ]
        }]
      }
      {
       type=MovieSceneFloatTrack
       display_name=Render Opacity
       sections [{type=MovieSceneFloatSection, channels [None {type=MovieSceneScriptingFloatChannel, keys [[0,0],[20,1]]}]}]
      }
     ]
    }
   ]
   display_rate {numerator=100}
  }
 }]
graphs [
 EventGraph {nodes [
   n1 = UpdateSortAndFilterButtons!() -> (OutputDelegate, then, bBladeOpen:bool=~) {is=title K2Node_CustomEvent}
   > n2 = GetAllSortAndFilterButtons!(execute, self:Athena Customization Picker Sort And Filter Blade Object Reference) -> (then, ReturnValue:Array of User Widget Object References)
   > n3 = For Each Loop(Array:Array of User Widget Object References=@n2) -> (Array Element:User Widget Object Reference)
   > n4 = Cast To SortAndFilterButton!(execute, Object=@n3) -> (then, CastFailed, AsSort and Filter Button:Sort And Filter Button Object Reference) {is=title K2Node_DynamicCast}
   > HandleBladeOpenChanged!(execute, self:Sort And Filter Button Object Reference=@n4, bBladeOpen:bool=@n1.bBladeOpen) -> (then)
   Event On Blade Open() {is=title K2Node_Event}
   > PlayAnimationForward(InAnimation=Get Anim_Open() -> (Anim_Open:Widget Animation Object Reference))
   > UpdateSortAndFilterButtons!(execute, self, bBladeOpen:bool=true) -> (then) {is=title K2Node_CallFunction}
   Event On Blade Close() {is=title K2Node_Event}
   n5 = Get Anim_Open()
   n6 = GetEndTime(self=@n5)
   n7 = GetAnimationCurrentTime!(self:User Widget Object Reference, InAnimation:Widget Animation Object Reference=@n5) -> (ReturnValue:float)
   > n8 = "Play Animation with Finished event (legacy)"!(
    execute:Exec
    Widget:User Widget Object Reference=Self-Reference()
    InAnimation:Widget Animation Object Reference=Get Anim_Open()
    StartAtTime:float=Select(Option 0:double=@n7, Option 1:double=float - float(A=@n6, B=@n7) {is=title K2Node_CallFunction}, Index=IsAnimationPlaying(InAnimation=@n5)) -> (ReturnValue:double)
    NumLoopsToPlay:int=1
    PlayMode:EUMGSequencePlayMode Enum=Reverse
    PlaybackSpeed:float=1
   ) -> (then:Exec, Result:UMGSequence Player Object Reference, Finished:Exec) {is=title K2Node_PlayAnimation}
   > UpdateSortAndFilterButtons(bBladeOpen=false)
   HandleCloseAnimationFinished(execute=@n8.Finished, self:Athena Customization Picker Sort And Filter Blade Object Reference)
   n9 = Event Set Sort Or Filter Button Display Name!() -> (OutputDelegate, then, SortOrFilterButton:Common Button Base Object Reference, DisplayName:"Text (by ref)") {is=title K2Node_Event}
   > n10 = Cast To SortAndFilterButton(Object=@n9.SortOrFilterButton)
   > SetText!(execute, self:Sort And Filter Button Object Reference=@n10, Text:Text=@n9.DisplayName) -> (then)
   n11 = Event Highlight Sort Or Filter Button!() -> (OutputDelegate, then, SortOrFilterButton:Common Button Base Object Reference, bHighlight:bool) {is=title K2Node_Event}
   > n12 = Cast To SortAndFilterButton(Object=@n11.SortOrFilterButton)
   > SetHighlightBorderVisibility!(execute, self:Sort And Filter Button Object Reference=@n12, bVisible:bool=@n11.bHighlight) -> (then)
   n13 = Event On Archive Filtering Mode Set!() -> (OutputDelegate, then, bEnabled:bool) {is=title K2Node_Event}
   > "SetText (Text)"(
    self=Get FilterByLabel() -> (FilterByLabel:Common Text Object Reference)
    InText=Select(Option 0:Text="LOCTEXT(\"Filter By\")", Option 1:Text="LOCTEXT(\"Filter Archived By\")", Index=@n13.bEnabled) -> (ReturnValue:Text)
   )
   > "SetText (Text)"(
    self=Get SortByLabel() -> (SortByLabel:Common Text Object Reference)
    InText=Select(Option 0="LOCTEXT(\"Sort By\")", Option 1="LOCTEXT(\"Sort Archived By\")", Index=@n13.bEnabled)
   )
  ]}
 NavigateDownFromSortButtons {nodes [
   n1 = NavigateDownFromSortButtons!() -> (then, Navigation:EUINavigation Enum) {is=title K2Node_FunctionEntry}
   > n2 = Switch on EUINavigation!(execute, Selection:EUINavigation Enum=@n1) -> (Left:Exec, Right:Exec, Up:Exec, Down:Exec, Next:Exec, Previous:Exec, Invalid:Exec) {is=title K2Node_SwitchEnum}
   Return Node!(
    execute=@n2.Down
    ReturnValue:Widget Object Reference=GetFilterButtonAtIndex!(self:Athena Customization Picker Sort And Filter Blade Object Reference, DesiredIndex:int=0) -> (ReturnValue:Common Button Base Object Reference)
   )
  ]}
 NavigateUpFromFilterButtons {nodes [
   n1 = NavigateUpFromFilterButtons!() -> (then, Navigation:EUINavigation Enum) {is=title K2Node_FunctionEntry}
   > n2 = Switch on EUINavigation(Selection=@n1)
   n3 = GetNumSorts!(self:Athena Customization Picker Sort And Filter Blade Object Reference) -> (ReturnValue:int=0)
   Return Node(
    execute=@n2.Up
    ReturnValue=GetSortButtonAtIndex!(
     self:Athena Customization Picker Sort And Filter Blade Object Reference
     DesiredIndex:int=int - int(A=@n3, B=1) {is=title K2Node_CallFunction}
    ) -> (ReturnValue:Common Button Base Object Reference)
   )
  ]}
]
variables []
widgets {
 info {
  inheritedWidgetCount=4
  parentClass=/Script/FortniteUI.AthenaCustomizationPickerSortAndFilterBlade
  rootWidgetClass=&Overlay
  widgetCount=18
 }
 widgets [
  ClippingBoundary_Overlay {widgetClassPath=&Overlay, widget=ClippingBoundary_Overlay, properties {clipping=ClipToBounds}}
  Border {
   widgetClassPath=&Border
   parent=ClippingBoundary_Overlay
   widget=Border
   properties {
    visibility=SelfHitTestInvisible
    background {
     bIsDynamicallyLoaded=false
     drawAs=Image
     imageType=NoImage
     mirroring=NoMirror
     resourceName=None
     resourceObject=null
     tiling=NoTile
     imageSize=<32 32>
     margin=<ltrb 0 0 0 0>
     outlineSettings {
      bUseBrushTransparency=false
      roundingType=HalfHeightRadius
      width=0
      color {colorUseRule=UseColor_Specified, specifiedColor=<rgba 0 0 0 0>}
      cornerRadii=<0 0 0 0>
     }
     tintColor {colorUseRule=UseColor_Specified, specifiedColor=<rgba 0.006 0.0103 0.2346 0.8>}
     uVRegion {bIsValid=false, max=<0 0>, min=<0 0>}
    }
    padding=<ltrb 0 0 0 0>
   }
   slot {class=OverlaySlot, properties {horizontalAlignment=HAlign_Fill, verticalAlignment=VAlign_Fill}}
  }
  Overlay_0 {widgetClassPath=&Overlay, parent=Border, widget=Overlay_0, slot {class=BorderSlot, properties {padding=<ltrb 0 0 0 0>}}}
  BackgroundBlur_30 {
   widgetClassPath=&BackgroundBlur
   parent=Overlay_0
   widget=BackgroundBlur_30
   properties {blurStrength=2}
   slot {class=OverlaySlot, properties {horizontalAlignment=HAlign_Fill, verticalAlignment=VAlign_Fill}}
  }
  AthenaScrollBox_52 {
   widgetClassPath=/Script/FortniteUI.AthenaScrollBox
   parent=Overlay_0
   widget=AthenaScrollBox_52
   properties {
    consumeMouseWheel=Always
    navigationScrollPadding=200
    scrollWhenFocusChanges=AnimatedScroll
    scrollbarPadding=<ltrb 0 0 18 0>
    widgetBarStyle {
     thickness=16
     draggedThumbImage {
      bIsDynamicallyLoaded=false
      drawAs=Image
      imageType=FullColor
      mirroring=NoMirror
      resourceName=../../../Engine/Content/Slate/Common/Scrollbar_Thumb.png
      resourceObject=/FortUILibrary/Foundation/Art_ToolKit/AngledBoxes/M_UI_ItemShop_ScrollBar_Drag
      tiling=NoTile
      imageSize=<0 0>
      margin=<ltrb 0.25 0.25 0.25 0.25>
      outlineSettings {
       bUseBrushTransparency=false
       roundingType=HalfHeightRadius
       width=0
       color {colorUseRule=UseColor_Specified, specifiedColor=<rgba 0 0 0 0>}
       cornerRadii=<0 0 0 0>
      }
      tintColor {colorUseRule=UseColor_Specified, specifiedColor=<rgba 1 1 1 1>}
      uVRegion {bIsValid=false, max=<0 0>, min=<0 0>}
     }
     horizontalBackgroundImage {
      bIsDynamicallyLoaded=false
      drawAs=NoDrawType
      imageType=NoImage
      mirroring=NoMirror
      resourceName=None
      resourceObject=null
      tiling=NoTile
      imageSize=<0 0>
      margin=<ltrb 0 0 0 0>
      outlineSettings {
       bUseBrushTransparency=false
       roundingType=HalfHeightRadius
       width=0
       color {colorUseRule=UseColor_Specified, specifiedColor=<rgba 0 0 0 0>}
       cornerRadii=<0 0 0 0>
      }
      tintColor {colorUseRule=UseColor_Specified, specifiedColor=<rgba 1 1 1 1>}
      uVRegion {bIsValid=false, max=<0 0>, min=<0 0>}
     }
     horizontalBottomSlotImage {
      bIsDynamicallyLoaded=false
      drawAs=NoDrawType
      imageType=FullColor
      mirroring=NoMirror
      resourceName=../../../Engine/Content/Slate/Common/Scrollbar_Background_Horizontal.png
      resourceObject=null
      tiling=NoTile
      imageSize=<8 8>
      margin=<ltrb 0 0 0 0>
      outlineSettings {
       bUseBrushTransparency=false
       roundingType=HalfHeightRadius
       width=0
       color {colorUseRule=UseColor_Specified, specifiedColor=<rgba 0 0 0 0>}
       cornerRadii=<0 0 0 0>
      }
      tintColor {colorUseRule=UseColor_Specified, specifiedColor=<rgba 1 1 1 1>}
      uVRegion {bIsValid=false, max=<0 0>, min=<0 0>}
     }
     horizontalTopSlotImage {
      bIsDynamicallyLoaded=false
      drawAs=NoDrawType
      imageType=FullColor
      mirroring=NoMirror
      resourceName=../../../Engine/Content/Slate/Common/Scrollbar_Background_Horizontal.png
      resourceObject=null
      tiling=NoTile
      imageSize=<8 8>
      margin=<ltrb 0 0 0 0>
      outlineSettings {
       bUseBrushTransparency=false
       roundingType=HalfHeightRadius
       width=0
       color {colorUseRule=UseColor_Specified, specifiedColor=<rgba 0 0 0 0>}
       cornerRadii=<0 0 0 0>
      }
      tintColor {colorUseRule=UseColor_Specified, specifiedColor=<rgba 1 1 1 1>}
      uVRegion {bIsValid=false, max=<0 0>, min=<0 0>}
     }
     hoveredThumbImage {
      bIsDynamicallyLoaded=false
      drawAs=Image
      imageType=FullColor
      mirroring=NoMirror
      resourceName=../../../Engine/Content/Slate/Common/Scrollbar_Thumb.png
      resourceObject=/FortUILibrary/Foundation/Art_ToolKit/AngledBoxes/M_UI_ItemShop_ScrollBar_Base
      tiling=NoTile
      imageSize=<0 0>
      margin=<ltrb 0.25 0.25 0.25 0.25>
      outlineSettings {
       bUseBrushTransparency=false
       roundingType=HalfHeightRadius
       width=0
       color {colorUseRule=UseColor_Specified, specifiedColor=<rgba 0 0 0 0>}
       cornerRadii=<0 0 0 0>
      }
      tintColor {colorUseRule=UseColor_Specified, specifiedColor=<rgba 1 1 1 1>}
      uVRegion {bIsValid=false, max=<0 0>, min=<0 0>}
     }
     normalThumbImage {
      bIsDynamicallyLoaded=false
      drawAs=Image
      imageType=FullColor
      mirroring=NoMirror
      resourceName=../../../Engine/Content/Slate/Common/Scrollbar_Thumb.png
      resourceObject=/Game/Athena/UI/Customization/Art/M_UI_LockerSortAndFilter_ScrollBar_Base
      tiling=NoTile
      imageSize=<0 0>
      margin=<ltrb 0.25 0.25 0.25 0.25>
      outlineSettings {
       bUseBrushTransparency=false
       roundingType=HalfHeightRadius
       width=0
       color {colorUseRule=UseColor_Specified, specifiedColor=<rgba 0 0 0 0>}
       cornerRadii=<0 0 0 0>
      }
      tintColor {colorUseRule=UseColor_Specified, specifiedColor=<rgba 1 1 1 1>}
      uVRegion {bIsValid=false, max=<0 0>, min=<0 0>}
     }
     verticalBackgroundImage {
      bIsDynamicallyLoaded=false
      drawAs=Image
      imageType=NoImage
      mirroring=NoMirror
      resourceName=None
      resourceObject=/Game/Athena/UI/Customization/Art/M_UI_Locker_ScrollBar_Backing
      tiling=NoTile
      imageSize=<0 0>
      margin=<ltrb 0 0 0 0>
      outlineSettings {
       bUseBrushTransparency=false
       roundingType=HalfHeightRadius
       width=0
       color {colorUseRule=UseColor_Specified, specifiedColor=<rgba 0 0 0 0>}
       cornerRadii=<0 0 0 0>
      }
      tintColor {colorUseRule=UseColor_Specified, specifiedColor=<rgba 1 1 1 1>}
      uVRegion {bIsValid=false, max=<0 0>, min=<0 0>}
     }
     verticalBottomSlotImage {
      bIsDynamicallyLoaded=false
      drawAs=NoDrawType
      imageType=FullColor
      mirroring=NoMirror
      resourceName=../../../Engine/Content/Slate/Common/Scrollbar_Background_Vertical.png
      resourceObject=null
      tiling=NoTile
      imageSize=<8 8>
      margin=<ltrb 0 0 0 0>
      outlineSettings {
       bUseBrushTransparency=false
       roundingType=HalfHeightRadius
       width=0
       color {colorUseRule=UseColor_Specified, specifiedColor=<rgba 0 0 0 0>}
       cornerRadii=<0 0 0 0>
      }
      tintColor {colorUseRule=UseColor_Specified, specifiedColor=<rgba 1 1 1 1>}
      uVRegion {bIsValid=false, max=<0 0>, min=<0 0>}
     }
     verticalTopSlotImage {
      bIsDynamicallyLoaded=false
      drawAs=NoDrawType
      imageType=FullColor
      mirroring=NoMirror
      resourceName=../../../Engine/Content/Slate/Common/Scrollbar_Background_Vertical.png
      resourceObject=null
      tiling=NoTile
      imageSize=<8 8>
      margin=<ltrb 0 0 0 0>
      outlineSettings {
       bUseBrushTransparency=false
       roundingType=HalfHeightRadius
       width=0
       color {colorUseRule=UseColor_Specified, specifiedColor=<rgba 0 0 0 0>}
       cornerRadii=<0 0 0 0>
      }
      tintColor {colorUseRule=UseColor_Specified, specifiedColor=<rgba 1 1 1 1>}
      uVRegion {bIsValid=false, max=<0 0>, min=<0 0>}
     }
    }
    widgetStyle {
     barThickness=9
     bottomShadowBrush {
      bIsDynamicallyLoaded=false
      drawAs=NoDrawType
      imageType=FullColor
      mirroring=NoMirror
      resourceName=../../../Engine/Content/Slate/Common/ScrollBoxShadowBottom.png
      resourceObject=null
      tiling=NoTile
      imageSize=<16 8>
      margin=<ltrb 0.5 0 0.5 1>
      outlineSettings {
       bUseBrushTransparency=false
       roundingType=HalfHeightRadius
       width=0
       color {colorUseRule=UseColor_Specified, specifiedColor=<rgba 0 0 0 0>}
       cornerRadii=<0 0 0 0>
      }
      tintColor {colorUseRule=UseColor_Specified, specifiedColor=<rgba 1 1 1 1>}
      uVRegion {bIsValid=false, max=<0 0>, min=<0 0>}
     }
     horizontalScrolledContentPadding=<ltrb 0 0 1 0>
     leftShadowBrush {
      bIsDynamicallyLoaded=false
      drawAs=NoDrawType
      imageType=FullColor
      mirroring=NoMirror
      resourceName=../../../Engine/Content/Slate/Common/ScrollBoxShadowLeft.png
      resourceObject=null
      tiling=NoTile
      imageSize=<8 16>
      margin=<ltrb 1 0.5 0 0.5>
      outlineSettings {
       bUseBrushTransparency=false
       roundingType=HalfHeightRadius
       width=0
       color {colorUseRule=UseColor_Specified, specifiedColor=<rgba 0 0 0 0>}
       cornerRadii=<0 0 0 0>
      }
      tintColor {colorUseRule=UseColor_Specified, specifiedColor=<rgba 1 1 1 1>}
      uVRegion {bIsValid=false, max=<0 0>, min=<0 0>}
     }
     rightShadowBrush {
      bIsDynamicallyLoaded=false
      drawAs=NoDrawType
      imageType=FullColor
      mirroring=NoMirror
      resourceName=../../../Engine/Content/Slate/Common/ScrollBoxShadowRight.png
      resourceObject=null
      tiling=NoTile
      imageSize=<8 16>
      margin=<ltrb 0 0.5 1 0.5>
      outlineSettings {
       bUseBrushTransparency=false
       roundingType=HalfHeightRadius
       width=0
       color {colorUseRule=UseColor_Specified, specifiedColor=<rgba 0 0 0 0>}
       cornerRadii=<0 0 0 0>
      }
      tintColor {colorUseRule=UseColor_Specified, specifiedColor=<rgba 1 1 1 1>}
      uVRegion {bIsValid=false, max=<0 0>, min=<0 0>}
     }
     topShadowBrush {
      bIsDynamicallyLoaded=false
      drawAs=NoDrawType
      imageType=FullColor
      mirroring=NoMirror
      resourceName=../../../Engine/Content/Slate/Common/ScrollBoxShadowTop.png
      resourceObject=null
      tiling=NoTile
      imageSize=<16 8>
      margin=<ltrb 0.5 1 0.5 0>
      outlineSettings {
       bUseBrushTransparency=false
       roundingType=HalfHeightRadius
       width=0
       color {colorUseRule=UseColor_Specified, specifiedColor=<rgba 0 0 0 0>}
       cornerRadii=<0 0 0 0>
      }
      tintColor {colorUseRule=UseColor_Specified, specifiedColor=<rgba 1 1 1 1>}
      uVRegion {bIsValid=false, max=<0 0>, min=<0 0>}
     }
     verticalScrolledContentPadding=<ltrb 0 0 0 1>
    }
   }
   slot {class=OverlaySlot, properties {horizontalAlignment=HAlign_Fill, verticalAlignment=VAlign_Fill}}
  }
  VerticalBox_42 {
   widgetClassPath=&VerticalBox
   parent=AthenaScrollBox_52
   widget=VerticalBox_42
   slot {class=ScrollBoxSlot, properties {padding=<ltrb 0 115 0 0>}}
  }
  Widget_Sorts {
   widgetClassPath=&VerticalBox
   bInherited=true
   bIsVariable=true
   parent=VerticalBox_42
   widget=Widget_Sorts
   slot {class=VerticalBoxSlot}
  }
  Border_0 {
   widgetClassPath=&Border
   parent=Widget_Sorts
   widget=Border_0
   properties {horizontalAlignment=HAlign_Left, brushColor=<rgba 0.0006 0.0006 0.0595 1>, padding=<ltrb 95 18 16 18>}
   slot {class=VerticalBoxSlot}
  }
  HorizontalBox_1 {
   widgetClassPath=&HorizontalBox
   parent=Border_0
   widget=HorizontalBox_1
   slot {class=BorderSlot, properties {horizontalAlignment=HAlign_Left, padding=<ltrb 95 18 16 18>}}
  }
  SortByLabel {
   widgetClassPath=&CommonTextBlock
   bIsVariable=true
   parent=HorizontalBox_1
   widget=SortByLabel
   properties {
    scrollStyle=&TextScrollStyle-Base_C
    text=Sort By
    textTransformPolicy=ToUpper
    colorAndOpacity {colorUseRule=UseColor_Specified, specifiedColor=<rgba 0 0.0823 0.9216 1>}
    font {
     bForceMonospaced=false
     bMaterialIsStencil=false
     fontMaterial=null
     fontObject=&BurbankSmall
     letterSpacing=0
     monospacedWidth=1
     size=21
     skewAmount=0
     typefaceFontName=Black
     outlineSettings {
      bApplyOutlineToDropShadows=false
      bMiteredCorners=false
      bSeparateFillAlpha=false
      outlineMaterial=null
      outlineSize=0
      outlineColor=<rgba 0 0 0 1>
     }
    }
    renderTransform {angle=0, scale=<1 1>, shear=<-8 0>, translation=<0 0>}
    shadowOffset=<0 0>
   }
   slot {class=HorizontalBoxSlot, properties {padding=<ltrb 0 0 8 0>, size {sizeRule=Fill, value=1}}}
  }
  Image_190 {
   widgetClassPath=&Image
   parent=HorizontalBox_1
   widget=Image_190
   properties {
    brush {
     bIsDynamicallyLoaded=false
     drawAs=Image
     imageType=NoImage
     mirroring=NoMirror
     resourceName=None
     resourceObject=/Game/UI/Foundation/Textures/Icons/Locker/T_Icon_Sort
     tiling=NoTile
     imageSize=<28 28>
     margin=<ltrb 0 0 0 0>
     outlineSettings {
      bUseBrushTransparency=false
      roundingType=HalfHeightRadius
      width=0
      color {colorUseRule=UseColor_Specified, specifiedColor=<rgba 0 0 0 0>}
      cornerRadii=<0 0 0 0>
     }
     tintColor {colorUseRule=UseColor_Specified, specifiedColor=<rgba 1 1 1 1>}
     uVRegion {bIsValid=false, max=<0 0>, min=<0 0>}
    }
    colorAndOpacity=<rgba 0 0.0823 0.9216 1>
   }
   slot {class=HorizontalBoxSlot, properties {verticalAlignment=VAlign_Center}}
  }
  DynamicEntryBox_SortButtons {
   widgetClassPath=/Script/FortniteUI.FortDynamicEntryBox
   bInherited=true
   bIsVariable=true
   parent=Widget_Sorts
   widget=DynamicEntryBox_SortButtons
   properties {
    entryBoxType=Vertical
    entryWidgetClass=&/SortAndFilterButton.SortAndFilterButton_C
    navigation=&self:WidgetTree.DynamicEntryBox_SortButtons.WidgetNavigation_0
   }
   slot {class=VerticalBoxSlot}
  }
  Widget_Filters {
   widgetClassPath=&VerticalBox
   bInherited=true
   bIsVariable=true
   parent=AthenaScrollBox_52
   widget=Widget_Filters
   slot {class=ScrollBoxSlot}
  }
  Border_1 {
   widgetClassPath=&Border
   parent=Widget_Filters
   widget=Border_1
   properties {horizontalAlignment=HAlign_Left, brushColor=<rgba 0.0006 0.0006 0.0595 1>, padding=<ltrb 95 18 16 18>}
   slot {class=VerticalBoxSlot}
  }
  HorizontalBox_0 {
   widgetClassPath=&HorizontalBox
   parent=Border_1
   widget=HorizontalBox_0
   slot {class=BorderSlot, properties {horizontalAlignment=HAlign_Left, padding=<ltrb 95 18 16 18>}}
  }
  FilterByLabel {
   widgetClassPath=&CommonTextBlock
   bIsVariable=true
   parent=HorizontalBox_0
   widget=FilterByLabel
   properties {
    scrollStyle=&TextScrollStyle-Base_C
    text=Filter By
    textTransformPolicy=ToUpper
    colorAndOpacity {colorUseRule=UseColor_Specified, specifiedColor=<rgba 0 0.0823 0.9216 1>}
    font {
     bForceMonospaced=false
     bMaterialIsStencil=false
     fontMaterial=null
     fontObject=&BurbankSmall
     letterSpacing=0
     monospacedWidth=1
     size=21
     skewAmount=0
     typefaceFontName=Black
     outlineSettings {
      bApplyOutlineToDropShadows=false
      bMiteredCorners=false
      bSeparateFillAlpha=false
      outlineMaterial=null
      outlineSize=0
      outlineColor=<rgba 0 0 0 1>
     }
    }
    renderTransform {angle=0, scale=<1 1>, shear=<-8 0>, translation=<0 0>}
    shadowOffset=<0 0>
   }
   slot {class=HorizontalBoxSlot, properties {padding=<ltrb 0 0 8 0>, size {sizeRule=Fill, value=1}}}
  }
  FilterIcon {
   widgetClassPath=&Image
   parent=HorizontalBox_0
   widget=FilterIcon
   properties {
    visibility=SelfHitTestInvisible
    brush {
     bIsDynamicallyLoaded=false
     drawAs=Image
     imageType=NoImage
     mirroring=NoMirror
     resourceName=None
     resourceObject=/Game/UI/Foundation/Textures/Icons/Locker/T_Icon_Filter
     tiling=NoTile
     imageSize=<28 28>
     margin=<ltrb 0 0 0 0>
     outlineSettings {
      bUseBrushTransparency=false
      roundingType=HalfHeightRadius
      width=0
      color {colorUseRule=UseColor_Specified, specifiedColor=<rgba 0 0 0 0>}
      cornerRadii=<0 0 0 0>
     }
     tintColor {colorUseRule=UseColor_Specified, specifiedColor=<rgba 1 1 1 1>}
     uVRegion {bIsValid=false, max=<0 0>, min=<0 0>}
    }
    colorAndOpacity=<rgba 0 0.0823 0.9216 1>
   }
   slot {class=HorizontalBoxSlot, properties {verticalAlignment=VAlign_Center}}
  }
  DynamicEntryBox_FilterButtons {
   widgetClassPath=&FortDynamicEntryBox
   bInherited=true
   bIsVariable=true
   parent=Widget_Filters
   widget=DynamicEntryBox_FilterButtons
   properties {
    entryBoxType=Vertical
    entryWidgetClass=&SortAndFilterButton_C
    navigation=&self:WidgetTree.DynamicEntryBox_FilterButtons.WidgetNavigation_0
   }
   slot {class=VerticalBoxSlot}
  }
 ]
}
