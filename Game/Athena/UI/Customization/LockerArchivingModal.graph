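# Locker archiving confirmation modal, a widget subclass of the C++ LockerConfirmationModal with a blurred lightbox, title, description and Confirm/Cancel buttons.
# The OnSetButtonTexts event casts both buttons to GenericModalButton and sets their text, falling back to a localized "Cancel" when the cancel text is empty.
# OnShowMTXWallet toggles the top-right container holding the MTX wallet button and player name.
# A close button beside it shows only for touch input.
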WidgetBlueprint /Game/Athena/UI/Customization/LockerArchivingModal
graphs [EventGraph {nodes [
   n1 = Event OnShowMTXWallet!() -> (OutputDelegate, then, bShow:bool) {is=title K2Node_Event}
   > SetVisibility(
    self=Get VBox_MtxContainer() -> (VBox_MtxContainer:Vertical Box Object Reference)
    InVisibility=Select(Option 0=Collapsed, Option 1=HitTestInvisible, Index=@n1.bShow)
   )
   n2 = Event OnSetButtonTexts!() -> (OutputDelegate, then, ConfirmText:"Text (by ref)", CancelText:"Text (by ref)") {is=title K2Node_Event}
   > n3 = Cast To GenericModalButton!(execute, Object=Get Button_Confirm() -> (Button_Confirm:Common Button Base Object Reference)) -> (then, CastFailed, AsGeneric Modal Button:Generic Modal Button Object Reference) {is=title K2Node_DynamicCast}
   > SetButtonText!(execute, self:Generic Modal Button Object Reference=@n3, ButtonText:Text=@n2.ConfirmText) -> (then)
   > n4 = Cast To GenericModalButton(Object=Get Button_Cancel() -> (Button_Cancel:Common Button Base Object Reference))
   > SetButtonText(
    self=@n4
    ButtonText=Select(Option 0:Text=@n2.CancelText, Option 1:Text="LOCTEXT(\"Cancel\")", Index=TextIsEmpty(InText=@n2.CancelText)) -> (ReturnValue:Text)
   )
   Event PreConstruct()
   Event Construct()
   Event Tick()
  ]}]
variables []
widgets {
 info {inheritedWidgetCount=5, parentClass=/Script/FortniteUI.LockerConfirmationModal, rootWidgetClass=&Overlay, widgetCount=27}
 widgets [
  Overlay {widgetClassPath=&Overlay, widget=Overlay}
  ContentVB {
   widgetClassPath=&VerticalBox
   parent=Overlay
   widget=ContentVB
   slot {class=OverlaySlot, properties {horizontalAlignment=HAlign_Fill}}
  }
  BackgroundBlurLightbox {
   widgetClassPath=&BackgroundBlur
   parent=Overlay
   widget=BackgroundBlurLightbox
   properties {
    blurStrength=8
    lowQualityFallbackBrush {
     bIsDynamicallyLoaded=false
     drawAs=Image
     imageType=NoImage
     mirroring=NoMirror
     resourceName=None
     resourceObject=null
     tiling=NoTile
     imageSize=<0 0>
     margin=<ltrb 0 0 0 0>
     outlineSettings {
      bUseBrushTransparency=false
      roundingType=HalfHeightRadius
      width=0
      color {colorUseRule=UseColor_Specified, specifiedColor=<rgba 0 0 0 0>}
      cornerRadii=<0 0 0 0>
     }
     tintColor {colorUseRule=UseColor_Specified, specifiedColor=<rgba 0 0 0 0.75>}
     uVRegion {bIsValid=false, max=<0 0>, min=<0 0>}
    }
   }
   slot {class=OverlaySlot, properties {horizontalAlignment=HAlign_Fill, verticalAlignment=VAlign_Fill}}
  }
  Lightbox_HitTestBlocker {
   widgetClassPath=&Image
   bIsVariable=true
   parent=BackgroundBlurLightbox
   widget=Lightbox_HitTestBlocker
   properties {brush {
     bIsDynamicallyLoaded=false
     drawAs=Image
     imageType=NoImage
     mirroring=NoMirror
     resourceName=None
     resourceObject=/Game/Athena/UI/Frontend/Art/M_UI_RadialGradientMaster_Dark
     tiling=NoTile
     imageSize=<32 32>
     margin=<ltrb 0 0 0 0>
     outlineSettings {
      bUseBrushTransparency=false
      roundingType=HalfHeightRadius
      width=0
      color {colorUseRule=UseColor_Specified, specifiedColor=<rgba 0 0 0 0>}
      cornerRadii=<0 0 0 0>
     }
     tintColor {colorUseRule=UseColor_Specified, specifiedColor=<rgba 0.5972 0.5972 0.5972 1>}
     uVRegion {bIsValid=false, max=<0 0>, min=<0 0>}
    }}
   slot {class=BackgroundBlurSlot, properties {padding=<ltrb 0 0 0 0>}}
  }
  PrimaryHeightSB {
   widgetClassPath=&SizeBox
   parent=Overlay
   widget=PrimaryHeightSB
   properties {bOverride_MinDesiredHeight=true, minDesiredHeight=350}
   slot {class=OverlaySlot, properties {horizontalAlignment=HAlign_Fill, verticalAlignment=VAlign_Center}}
  }
  Overlay_Modal {widgetClassPath=&Overlay, parent=PrimaryHeightSB, widget=Overlay_Modal, slot {class=SizeBoxSlot}}
  Containment {
   widgetClassPath=&Image
   bIsVariable=true
   parent=Overlay_Modal
   widget=Containment
   properties {
    brush {
     bIsDynamicallyLoaded=false
     drawAs=Image
     imageType=NoImage
     mirroring=NoMirror
     resourceName=None
     resourceObject=/Game/Athena/UI/Subscriptions/Art/M_UI_Subs_Color_FIll_Modals
     tiling=NoTile
     imageSize=<32 32>
     margin=<ltrb 0 0 0 0>
     outlineSettings {
      bUseBrushTransparency=false
      roundingType=HalfHeightRadius
      width=0
      color {colorUseRule=UseColor_Specified, specifiedColor=<rgba 0 0 0 0>}
      cornerRadii=<0 0 0 0>
     }
     tintColor {colorUseRule=UseColor_Specified, specifiedColor=<rgba 1 1 1 1>}
     uVRegion {bIsValid=false, max=<0 0>, min=<0 0>}
    }
    colorAndOpacity=<rgba 1 1 1 0.5>
    renderTransformPivot=<0.5 1>
   }
   slot {class=OverlaySlot, properties {horizontalAlignment=HAlign_Fill, verticalAlignment=VAlign_Fill}}
  }
  ContentSZ {
   widgetClassPath=&SafeZone
   bIsVariable=true
   parent=Overlay_Modal
   widget=ContentSZ
   slot {class=OverlaySlot, properties {horizontalAlignment=HAlign_Fill, verticalAlignment=VAlign_Fill, padding=<ltrb 64 0 64 0>}}
  }
  HorizontalBox_0 {widgetClassPath=&HorizontalBox, parent=ContentSZ, widget=HorizontalBox_0, slot {class=SafeZoneSlot}}
  Spacer_162 {
   widgetClassPath=&Spacer
   parent=HorizontalBox_0
   widget=Spacer_162
   slot {class=HorizontalBoxSlot, properties {size {sizeRule=Fill, value=0.15}}}
  }
  Overlay_2 {
   widgetClassPath=&Overlay
   parent=HorizontalBox_0
   widget=Overlay_2
   slot {class=HorizontalBoxSlot, properties {padding=<ltrb 0 75 0 150>, size {sizeRule=Fill, value=1}}}
  }
  TextContent {
   widgetClassPath=&VerticalBox
   parent=Overlay_2
   widget=TextContent
   slot {class=OverlaySlot, properties {horizontalAlignment=HAlign_Fill, verticalAlignment=VAlign_Fill}}
  }
  Text_Title {
   widgetClassPath=&CommonTextBlock
   bInherited=true
   bIsVariable=true
   parent=TextContent
   widget=Text_Title
   properties {
    autoWrapText=true
    justification=Center
    mobileFontSizeMultiplier=1.2
    text=Locker archiving confirmation modal text
    textTransformPolicy=ToUpper
    wrapTextAt=1152
    font {
     bForceMonospaced=false
     bMaterialIsStencil=false
     fontMaterial=null
     fontObject=&Burbank
     letterSpacing=0
     monospacedWidth=1
     size=30
     skewAmount=0
     typefaceFontName=BurbankRegular-Black
     outlineSettings {
      bApplyOutlineToDropShadows=false
      bMiteredCorners=false
      bSeparateFillAlpha=false
      outlineMaterial=null
      outlineSize=0
      outlineColor=<rgba 1 1 1 0>
     }
    }
    renderTransform {angle=0, scale=<1 1>, shear=<-12 0>, translation=<0 0>}
    shadowOffset=<0 0>
   }
   slot {class=VerticalBoxSlot, properties {padding=<ltrb 0 0 0 25>}}
  }
  Text_Description {
   widgetClassPath=&CommonTextBlock
   bInherited=true
   bIsVariable=true
   parent=TextContent
   widget=Text_Description
   properties {
    autoWrapText=true
    justification=Center
    mobileFontSizeMultiplier=1.2
    text=Locker archiving confirmation modal description.
    wrapTextAt=1152
    colorAndOpacity {colorUseRule=UseColor_Specified, specifiedColor=<rgba 0.0235 0.5922 0.8157 1>}
    font {
     bForceMonospaced=false
     bMaterialIsStencil=false
     fontMaterial=null
     fontObject=&BurbankSmall
     letterSpacing=0
     monospacedWidth=1
     size=22
     skewAmount=0
     typefaceFontName=BurbankRegular-Black
     outlineSettings {
      bApplyOutlineToDropShadows=false
      bMiteredCorners=false
      bSeparateFillAlpha=false
      outlineMaterial=null
      outlineSize=0
      outlineColor=<rgba 1 1 1 0>
     }
    }
    renderTransform {angle=0, scale=<1 1>, shear=<-12 0>, translation=<0 0>}
    shadowOffset=<0 0>
   }
   slot {class=VerticalBoxSlot}
  }
  Spacer_302 {
   widgetClassPath=&Spacer
   parent=HorizontalBox_0
   widget=Spacer_302
   slot {class=HorizontalBoxSlot, properties {horizontalAlignment=HAlign_Left, verticalAlignment=VAlign_Top, size {sizeRule=Fill, value=0.15}}}
  }
  Image_BackgroundBottomBar {
   widgetClassPath=&Image
   bIsVariable=true
   parent=Overlay_Modal
   widget=Image_BackgroundBottomBar
   properties {brush {
     bIsDynamicallyLoaded=false
     drawAs=Image
     imageType=NoImage
     mirroring=NoMirror
     resourceName=None
     resourceObject=/FortUILibrary/Foundation/Art_ToolKit/AngledBoxes/M_UI_MoreInfo_BottomDividerLIne_Blue
     tiling=NoTile
     imageSize=<32 26>
     margin=<ltrb 0 0 0 0>
     outlineSettings {
      bUseBrushTransparency=false
      roundingType=HalfHeightRadius
      width=0
      color {colorUseRule=UseColor_Specified, specifiedColor=<rgba 0 0 0 0>}
      cornerRadii=<0 0 0 0>
     }
     tintColor {colorUseRule=UseColor_Specified, specifiedColor=<rgba 0.3263 0.3457 0.4149 1>}
     uVRegion {bIsValid=false, max=<0 0>, min=<0 0>}
    }}
   slot {class=OverlaySlot, properties {horizontalAlignment=HAlign_Fill, verticalAlignment=VAlign_Bottom}}
  }
  ButtonHeightSB {
   widgetClassPath=&SizeBox
   parent=Overlay_Modal
   widget=ButtonHeightSB
   properties {bOverride_MaxDesiredHeight=true, maxDesiredHeight=120}
   slot {class=OverlaySlot, properties {horizontalAlignment=HAlign_Fill, verticalAlignment=VAlign_Bottom, padding=<ltrb 0 32 0 -44>}}
  }
  HorizontalBox_149 {
   widgetClassPath=&HorizontalBox
   parent=ButtonHeightSB
   widget=HorizontalBox_149
   slot {class=SizeBoxSlot, properties {horizontalAlignment=HAlign_Center, verticalAlignment=VAlign_Top}}
  }
  Button_Confirm {
   widgetClassPath=/Game/Athena/UI/GenericModalButton
   bInherited=true
   bIsVariable=true
   parent=HorizontalBox_149
   widget=Button_Confirm
   properties {
    auto wrap text=false
    button text="[PH] Confirm"
    fontSize=24
    fontSizeOverride=true
    material Brush=/Game/Athena/UI/Frontend/Spatial/Art/M_Ui_Skewed_SpatialButton_MoreInfo_Blue
    minWidth=192
    onButtonBaseClicked=(&self:WidgetTree.Button_Confirm.HandleOnButtonClicked)
    onButtonBaseDoubleClicked=(&self:WidgetTree.Button_Confirm.HandleOnButtonDoubleClicked)
    onButtonBaseHovered=(&self:WidgetTree.Button_Confirm.HandleOnButtonHovered)
    onButtonBaseUnhovered=(&self:WidgetTree.Button_Confirm.HandleOnButtonUnhovered)
    onSelectedChangedBase=(&self:WidgetTree.Button_Confirm.HandleOnSelectedChanged)
    right Image=null
    shearText=true
    simpleHover=true
    hoveredSlateSoundOverride {resourceObject=/Game/Sounds/Fort_UI_Sounds/FrontEnd/DefaultSounds/PressedButtons/MSS_UI_Button_Hover_Default}
    pressedSlateSoundOverride {resourceObject=/Game/Sounds/Fort_UI_Sounds/FrontEnd/New_FrontEnd/Fortnite_UI_MainMenu_Back_01_Cue}
    triggeringInputAction {dataTable=&InputActionDataTable, rowName=GenericForward}
   }
   slot {class=HorizontalBoxSlot, properties {horizontalAlignment=HAlign_Center, padding=<ltrb 0 0 10 0>}}
  }
  Button_Cancel {
   widgetClassPath=&GenericModalButton
   bInherited=true
   bIsVariable=true
   parent=HorizontalBox_149
   widget=Button_Cancel
   properties {
    auto wrap text=false
    button text="[PH] Cancel"
    fontSize=24
    fontSizeOverride=true
    material Brush=&M_Ui_Skewed_SpatialButton_MoreInfo_Blue
    minWidth=192
    onButtonBaseClicked=(&self:WidgetTree.Button_Cancel.HandleOnButtonClicked)
    onButtonBaseDoubleClicked=(&self:WidgetTree.Button_Cancel.HandleOnButtonDoubleClicked)
    onButtonBaseHovered=(&self:WidgetTree.Button_Cancel.HandleOnButtonHovered)
    onButtonBaseUnhovered=(&self:WidgetTree.Button_Cancel.HandleOnButtonUnhovered)
    onSelectedChangedBase=(&self:WidgetTree.Button_Cancel.HandleOnSelectedChanged)
    right Image=null
    shearText=true
    simpleHover=true
    hoveredSlateSoundOverride {resourceObject=&MSS_UI_Button_Hover_Default}
    pressedSlateSoundOverride {resourceObject=&Fortnite_UI_MainMenu_Back_01_Cue}
    triggeringInputAction {dataTable=&InputActionDataTable, rowName=GenericBack}
   }
   slot {class=HorizontalBoxSlot, properties {horizontalAlignment=HAlign_Center, padding=<ltrb 10 0 0 0>}}
  }
  HorizontalBox_1 {
   widgetClassPath=&HorizontalBox
   parent=Overlay
   widget=HorizontalBox_1
   slot {class=OverlaySlot, properties {horizontalAlignment=HAlign_Right, padding=<ltrb 0 32 32 0>}}
  }
  VBox_MtxContainer {
   widgetClassPath=&VerticalBox
   bIsVariable=true
   parent=HorizontalBox_1
   widget=VBox_MtxContainer
   slot {class=HorizontalBoxSlot, properties {padding=<ltrb 0 0 24 0>}}
  }
  Button_MtxWallet {
   widgetClassPath=/Game/UI/Frontend/TopBar/MTXButton
   parent=VBox_MtxContainer
   widget=Button_MtxWallet
   properties {
    onButtonBaseClicked=(&self:WidgetTree.Button_MtxWallet.HandleOnButtonClicked)
    onButtonBaseDoubleClicked=(&self:WidgetTree.Button_MtxWallet.HandleOnButtonDoubleClicked)
    onButtonBaseHovered=(&self:WidgetTree.Button_MtxWallet.HandleOnButtonHovered)
    onButtonBaseUnhovered=(&self:WidgetTree.Button_MtxWallet.HandleOnButtonUnhovered)
    onSelectedChangedBase=(&self:WidgetTree.Button_MtxWallet.HandleOnSelectedChanged)
    visibility=Collapsed
   }
   slot {class=VerticalBoxSlot, properties {horizontalAlignment=HAlign_Right}}
  }
  FortOwningPlayerSocialNameTextBlock {
   widgetClassPath=/Game/UI/Social/FortOwningPlayerSocialNameTextBlock
   parent=VBox_MtxContainer
   widget=FortOwningPlayerSocialNameTextBlock
   slot {class=VerticalBoxSlot, properties {horizontalAlignment=HAlign_Right, padding=<ltrb 0 4 0 0>}}
  }
  TouchClose {
   widgetClassPath=&SafeZone
   bIsVariable=true
   parent=HorizontalBox_1
   widget=TouchClose
   properties {padBottom=false, padLeft=false}
   slot {class=HorizontalBoxSlot}
  }
  VisibilityBorder_Touch {
   widgetClassPath=&CommonHardwareVisibilityBorder
   parent=TouchClose
   widget=VisibilityBorder_Touch
   properties {visibilityQuery {
     autoDescription=" ANY( Input.Touch )"
     tokenStreamVersion=0
     userDescription=VisibleUsingTouch
     queryTokenStream [0,1,1,1,0]
     tagDictionary [{tagName=Input.Touch}]
    }}
   slot {class=SafeZoneSlot}
  }
  Button_CloseTouch {
   widgetClassPath=/FortUILibrary/Foundation/Buttons/CloseButton
   bInherited=true
   bIsVariable=true
   parent=VisibilityBorder_Touch
   widget=Button_CloseTouch
   properties {
    onButtonBaseClicked=(&self:WidgetTree.Button_CloseTouch.HandleOnButtonClicked)
    onButtonBaseDoubleClicked=(&self:WidgetTree.Button_CloseTouch.HandleOnButtonDoubleClicked)
    onButtonBaseHovered=(&self:WidgetTree.Button_CloseTouch.HandleOnButtonHovered)
    onButtonBaseUnhovered=(&self:WidgetTree.Button_CloseTouch.HandleOnButtonUnhovered)
    onSelectedChangedBase=(&self:WidgetTree.Button_CloseTouch.HandleOnSelectedChanged)
   }
   slot {class=BorderSlot, properties {padding=<ltrb 0 0 0 0>}}
  }
 ]
}
